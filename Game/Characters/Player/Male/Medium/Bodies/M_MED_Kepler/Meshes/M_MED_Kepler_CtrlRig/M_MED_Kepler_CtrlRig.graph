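# Control Rig for the Kepler medium male player body that drives its extra digitigrade "wolf" leg bones from the standard leg bones.
# Per leg, thigh, calf and foot local rotations, adjusted by rest-pose offsets, set wolf thigh, calf and ball rotations.
# FABRIK then solves wolf thigh to toe toward the animated ball bone plus its rest offset, and the toe is snapped to that transform.
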ControlRigBlueprint /Game/Characters/Player/Male/Medium/Bodies/M_MED_Kepler/Meshes/M_MED_Kepler_CtrlRig/M_MED_Kepler_CtrlRig
graphs [
 RigVMFunctionLibrary {links [], nodes []}
 RigVMModel {nodes [
   n1 = &RigUnit_BeginExecution()
   n2 = /Script/ControlRig.RigUnit_GetInitialBoneTransform(Bone=thigh_r, Space=GlobalSpace) {is=struct RigVMUnitNode}
   n3 = &RigUnit_GetInitialBoneTransform(Bone=wolf_thigh_r, Space=GlobalSpace)
   /Script/ControlRig.RigUnit_SetBoneRotation(
    Bone=wolf_thigh_r
    Space=LocalSpace
    Weight=1
    bPropagateToChildren=True
    ExecutePin=@n1.ExecutePin
    Rotation=&RigVMFunction_MathTransformMakeRelative(
     Global=&RigUnit_GetBoneTransform(Bone=thigh_r, Space=LocalSpace).Transform
     Parent=&RigVMFunction_MathTransformMakeRelative(Global=@n2.Transform, Parent=@n3.Transform).Local
    ).Local.Rotation
   ) {is=struct RigVMUnitNode}
   n4 = &RigUnit_GetInitialBoneTransform(Bone=calf_r, Space=GlobalSpace)
   n5 = &RigUnit_GetInitialBoneTransform(Bone=wolf_calf_r, Space=GlobalSpace)
   > &RigUnit_SetBoneRotation(
    Bone=wolf_calf_r
    Space=LocalSpace
    Weight=1
    bPropagateToChildren=True
    Rotation=&RigVMFunction_MathTransformMakeRelative(
     Global=&RigUnit_GetBoneTransform(Bone=calf_r, Space=LocalSpace).Transform
     Parent=&RigVMFunction_MathTransformMakeRelative(Global=@n4.Transform, Parent=@n5.Transform).Local
    ).Local.Rotation
   )
   n6 = &RigUnit_GetInitialBoneTransform(Bone=foot_r, Space=GlobalSpace)
   n7 = &RigUnit_GetInitialBoneTransform(Bone=wolf_ball_r, Space=GlobalSpace)
   > &RigUnit_SetBoneRotation(
    Bone=wolf_ball_r
    Space=LocalSpace
    Weight=1
    bPropagateToChildren=True
    Rotation=&RigVMFunction_MathTransformMakeRelative(
     Global=&RigUnit_GetBoneTransform(Bone=foot_r, Space=LocalSpace).Transform
     Parent=&RigVMFunction_MathTransformMakeRelative(Global=@n6.Transform, Parent=@n7.Transform).Local
    ).Local.Rotation
   )
   n8 = &RigVMFunction_MathTransformMul(
    B=&RigUnit_GetBoneTransform(Bone=ball_r).Transform
    A=&RigVMFunction_MathTransformMakeRelative(
     Parent=&RigUnit_GetInitialBoneTransform(Bone=ball_r, Space=GlobalSpace).Transform
     Global=&RigUnit_GetInitialBoneTransform(Bone=wolf_toe_r, Space=GlobalSpace).Transform
    ).Local
   ) {template="Multiply::Execute(in A,in B,out Result)"}
   > /Script/ControlRig.RigUnit_FABRIK(
    EffectorBone=wolf_toe_r
    MaxIterations=10
    Precision=1
    StartBone=wolf_thigh_r
    Weight=1
    bPropagateToChildren=True
    bSetEffectorTransform=True
    EffectorTransform=@n8.Result
   ) {is=struct RigVMUnitNode}
   > /Script/ControlRig.RigUnit_SetBoneTransform(Bone=wolf_toe_r, Space=GlobalSpace, Weight=1, bPropagateToChildren=True, Transform=@n8.Result) {is=struct RigVMUnitNode}
   n9 = &RigUnit_GetInitialBoneTransform(Bone=wolf_thigh_l, Space=GlobalSpace)
   n10 = &RigUnit_GetInitialBoneTransform(Bone=thigh_l, Space=GlobalSpace)
   > &RigUnit_SetBoneRotation(
    Bone=wolf_thigh_l
    Space=LocalSpace
    Weight=1
    bPropagateToChildren=True
    Rotation=&RigVMFunction_MathTransformMakeRelative(
     Global=&RigUnit_GetBoneTransform(Bone=thigh_l, Space=LocalSpace).Transform
     Parent=&RigVMFunction_MathTransformMakeRelative(Parent=@n9.Transform, Global=@n10.Transform).Local
    ).Local.Rotation
   )
   n11 = &RigUnit_GetInitialBoneTransform(Bone=calf_l, Space=GlobalSpace)
   n12 = &RigUnit_GetInitialBoneTransform(Bone=wolf_calf_l, Space=GlobalSpace)
   > &RigUnit_SetBoneRotation(
    Bone=wolf_calf_l
    Space=LocalSpace
    Weight=1
    bPropagateToChildren=True
    Rotation=&RigVMFunction_MathTransformMakeRelative(
     Global=&RigUnit_GetBoneTransform(Bone=calf_l, Space=LocalSpace).Transform
     Parent=&RigVMFunction_MathTransformMakeRelative(Global=@n11.Transform, Parent=@n12.Transform).Local
    ).Local.Rotation
   )
   n13 = &RigUnit_GetInitialBoneTransform(Bone=foot_l, Space=GlobalSpace)
   n14 = &RigUnit_GetInitialBoneTransform(Bone=wolf_ball_l, Space=GlobalSpace)
   > &RigUnit_SetBoneRotation(
    Bone=wolf_ball_l
    Space=LocalSpace
    Weight=1
    bPropagateToChildren=True
    Rotation=&RigVMFunction_MathTransformMakeRelative(
     Global=&RigUnit_GetBoneTransform(Bone=foot_l, Space=LocalSpace).Transform
     Parent=&RigVMFunction_MathTransformMakeRelative(Global=@n13.Transform, Parent=@n14.Transform).Local
    ).Local.Rotation
   )
   n15 = &RigVMFunction_MathTransformMul(
    B=&RigUnit_GetBoneTransform(Bone=ball_l).Transform
    A=&RigVMFunction_MathTransformMakeRelative(
     Parent=&RigUnit_GetInitialBoneTransform(Bone=ball_l, Space=GlobalSpace).Transform
     Global=&RigUnit_GetInitialBoneTransform(Bone=wolf_toe_l, Space=GlobalSpace).Transform
    ).Local
   ) {template="Multiply::Execute(in A,in B,out Result)"}
   > &RigUnit_FABRIK(
    EffectorBone=wolf_toe_l
    MaxIterations=10
    Precision=1
    StartBone=wolf_thigh_l
    Weight=1
    bPropagateToChildren=True
    bSetEffectorTransform=True
    EffectorTransform=@n15.Result
   )
   > &RigUnit_SetBoneTransform(Bone=wolf_toe_l, Space=GlobalSpace, Weight=1, bPropagateToChildren=True, Transform=@n15.Result)
   &RigUnit_BeginExecution()
  ]}
]
hierarchy [
 root {loc=<0 0 0>, rot=<0 0 0 1>}
 pelvis {parent=root, loc=<0 -0.7071 93.1191>, rot=<0.0127 0.707 -0.0127 -0.707>}
 spine_01 {parent=pelvis, loc=<6.7494 -0.6388 0>, rot=<0 0 0.034 -0.9994>}
 spine_02 {parent=spine_01, loc=<7.2518 -0.035 0>, rot=<0 0 0.0238 -0.9997>}
 spine_03 {parent=spine_02, loc=<6.0104 0 0>, rot=<0 0 0.0149 -0.9999>}
 spine_04 {parent=spine_03, loc=<8.1024 0 0>, rot=<0 0 -0.0132 -0.9999>}
 spine_05 {parent=spine_04, loc=<10.0146 0 0>, rot=<0 0 -0.0015 -1>}
 clavicle_l {parent=spine_05, loc=<4.3575 -0.1466 -2.3223>, rot=<0.0465 -0.7289 -0.009 -0.6829>}
 upperarm_l {parent=clavicle_l, loc=<15.5111 0 0>, rot=<0.0226 -0.3234 -0.0088 -0.9459>}
 lowerarm_l {parent=upperarm_l, loc=<26.6514 0 0>, rot=<0.0391 -0.0099 -0.2453 -0.9686>}
 hand_l {parent=lowerarm_l, loc=<22.3491 0 0>, rot=<-0.7081 -0.0507 0.0975 -0.6975>}
 index_metacarpal_l {parent=hand_l, loc=<1.6535 0.0183 -2.96>, rot=<0.0158 -0.0124 -0.0173 -0.9996>}
 index_01_l {parent=index_metacarpal_l, loc=<8.3008 0.0357 -0.0524>, rot=<-0.0044 0.0485 0.0924 -0.9945>}
 index_02_l {parent=index_01_l, loc=<3.3587 -0.3004 0.0452>, rot=<-0.002 0.0027 0.3225 -0.9466>}
 index_03_l {parent=index_02_l, loc=<2.5578 -0.123 -0.0252>, rot=<-0.002 0.0067 0.2558 -0.9667>}
 middle_metacarpal_l {parent=hand_l, loc=<2.1856 -0.1725 -0.5984>, rot=<0.006 0.0168 -0.0329 -0.9993>}
 middle_01_l {parent=middle_metacarpal_l, loc=<7.3275 -0.2447 0.0129>, rot=<-0.0208 0.0042 0.0858 -0.9961>}
 middle_02_l {parent=middle_01_l, loc=<4.0123 0.1401 0.1175>, rot=<-0.0218 0.0258 0.3512 -0.9357>}
 middle_03_l {parent=middle_02_l, loc=<2.6077 0.0534 0.1373>, rot=<0.0082 0.0188 0.1708 -0.9851>}
 pinky_metacarpal_l {parent=hand_l, loc=<1.7616 -0.9753 3.3814>, rot=<-0.0397 0.0814 -0.069 -0.9935>}
 pinky_01_l {parent=pinky_metacarpal_l, loc=<7.1709 -0.4831 -0.1094>, rot=<-0.0654 -0.0035 0.2242 -0.9723>}
 pinky_02_l {parent=pinky_01_l, loc=<2.8274 0.0893 -0.1043>, rot=<0.0037 0.0011 0.2854 -0.9584>}
 pinky_03_l {parent=pinky_02_l, loc=<2.2046 0.0307 -0.0078>, rot=<-0.0002 0.0098 0.1445 -0.9895>}
 ring_metacarpal_l {parent=hand_l, loc=<2.2074 -0.3625 1.3638>, rot=<-0.0284 0.0722 -0.0458 -0.9959>}
 ring_01_l {parent=ring_metacarpal_l, loc=<7.0295 -0.6197 -0.1446>, rot=<-0.0489 -0.0252 0.1892 -0.9804>}
 ring_02_l {parent=ring_01_l, loc=<3.6174 0.4646 -0.1027>, rot=<0 0 0.2806 -0.9598>}
 ring_03_l {parent=ring_02_l, loc=<2.4152 0.1141 0.0199>, rot=<0 0 0.1189 -0.9929>}
 thumb_01_l {parent=hand_l, loc=<1.3118 -0.3191 -2.459>, rot=<0.6546 -0.4042 0.0664 -0.6354>}
 thumb_02_l {parent=thumb_01_l, loc=<5.9336 -0.0448 0.0401>, rot=<-0.0037 0.0213 0.298 -0.9543>}
 thumb_03_l {parent=thumb_02_l, loc=<2.2223 0.3682 -0.0907>, rot=<0.0118 -0.013 0.0066 -0.9998>}
 lowerarm_twist_01_l {parent=lowerarm_l, loc=<17.8549 0 0>, rot=<0 0 0 1>}
 lowerarm_twist_02_l {parent=lowerarm_l, loc=<9.6199 0 0>, rot=<0 0 0 1>}
 upperarm_twist_01_l {parent=upperarm_l, loc=<0.0041 0 0>, rot=<0 0 0 -1>}
 upperarm_twist_02_l {parent=upperarm_l, loc=<13.7941 0 0>, rot=<0 0 0 -1>}
 clavicle_r {parent=spine_05, loc=<4.3575 -0.1466 2.3223>, rot=<0.7289 0.0465 -0.6829 0.009>}
 upperarm_r {parent=clavicle_r, loc=<-15.5111 0 0>, rot=<0.0226 -0.3234 -0.0088 -0.9459>}
 lowerarm_r {parent=upperarm_r, loc=<-26.6514 0 0>, rot=<0.0391 -0.0099 -0.2453 -0.9686>}
 hand_r {parent=lowerarm_r, loc=<-22.3492 0 0>, rot=<-0.7081 -0.0507 0.0975 -0.6975>}
 index_metacarpal_r {parent=hand_r, loc=<-1.6534 -0.0181 2.9601>, rot=<0.0158 -0.0124 -0.0173 -0.9996>}
 index_01_r {parent=index_metacarpal_r, loc=<-8.3008 -0.0357 0.0524>, rot=<-0.0044 0.0485 0.0924 -0.9945>}
 index_02_r {parent=index_01_r, loc=<-3.3589 0.3003 -0.0453>, rot=<-0.002 0.0027 0.3225 -0.9466>}
 index_03_r {parent=index_02_r, loc=<-2.5578 0.1231 0.0251>, rot=<-0.002 0.0067 0.2558 -0.9667>}
 middle_metacarpal_r {parent=hand_r, loc=<-2.1856 0.1726 0.5984>, rot=<0.006 0.0168 -0.0329 -0.9993>}
 middle_01_r {parent=middle_metacarpal_r, loc=<-7.3273 0.2447 -0.0129>, rot=<-0.0208 0.0042 0.0858 -0.9961>}
 middle_02_r {parent=middle_01_r, loc=<-4.0142 -0.1411 -0.07>, rot=<-0.0217 0.0259 0.3512 -0.9357>}
 middle_03_r {parent=middle_02_r, loc=<-2.6074 -0.0534 -0.1312>, rot=<0.0083 0.0188 0.1708 -0.9851>}
 pinky_metacarpal_r {parent=hand_r, loc=<-1.7617 0.9753 -3.3814>, rot=<-0.0397 0.0814 -0.069 -0.9935>}
 pinky_01_r {parent=pinky_metacarpal_r, loc=<-7.1707 0.4832 0.1094>, rot=<-0.0654 -0.0035 0.2242 -0.9723>}
 pinky_02_r {parent=pinky_01_r, loc=<-2.8274 -0.0892 0.1043>, rot=<0.0037 0.0011 0.2854 -0.9584>}
 pinky_03_r {parent=pinky_02_r, loc=<-2.2045 -0.031 0.0078>, rot=<-0.0002 0.0098 0.1445 -0.9895>}
 ring_metacarpal_r {parent=hand_r, loc=<-2.2073 0.3627 -1.3638>, rot=<-0.0284 0.0722 -0.0458 -0.9959>}
 ring_01_r {parent=ring_metacarpal_r, loc=<-7.0297 0.6197 0.1446>, rot=<-0.0489 -0.0252 0.1892 -0.9804>}
 ring_02_r {parent=ring_01_r, loc=<-3.6172 -0.4648 0.1027>, rot=<0 0 0.2806 -0.9598>}
 ring_03_r {parent=ring_02_r, loc=<-2.4152 -0.1141 -0.0199>, rot=<0 0 0.1189 -0.9929>}
 thumb_01_r {parent=hand_r, loc=<-1.3119 0.3192 2.459>, rot=<0.6546 -0.4042 0.0664 -0.6354>}
 thumb_02_r {parent=thumb_01_r, loc=<-5.9335 0.0447 -0.0401>, rot=<-0.0037 0.0213 0.298 -0.9543>}
 thumb_03_r {parent=thumb_02_r, loc=<-2.2223 -0.3682 0.0907>, rot=<0.0118 -0.013 0.0066 -0.9998>}
 lowerarm_twist_01_r {parent=lowerarm_r, loc=<-17.855 0 0>, rot=<0 0 0 1>}
 lowerarm_twist_02_r {parent=lowerarm_r, loc=<-9.6199 0 0>, rot=<0 0 0 1>}
 upperarm_twist_01_r {parent=upperarm_r, loc=<-0.0041 0 0>, rot=<0 0 0 1>}
 upperarm_twist_02_r {parent=upperarm_r, loc=<-13.7942 0 0>, rot=<0 0 0 1>}
 neck_01 {parent=spine_05, loc=<11.3138 0.1529 0>, rot=<0 0 -0.2003 -0.9797>}
 neck_02 {parent=neck_01, loc=<4.5627 0 0>, rot=<0 0 0.0504 -0.9987>}
 thigh_l {parent=pelvis, loc=<-2.5693 -0.0109 -7.7501>, rot=<0.0012 0.0572 0.0155 -0.9982>}
 calf_l {parent=thigh_l, loc=<-39.0484 0 0>, rot=<0.0006 -0.0161 -0.0391 -0.9991>}
 foot_l {parent=calf_l, loc=<-42.561 0.1077 0.4292>, rot=<0.0418 -0.0455 0.0343 -0.9975>}
 ball_l {parent=foot_l, loc=<-6.1599 13.8264 0.6806>, rot=<0.0001 -0.0001 -0.7063 -0.7079>}
 thigh_twist_01_l {parent=thigh_l, loc=<-21.6303 0 0>, rot=<0.0013 0 0 -1>}
 thigh_r {parent=pelvis, loc=<-2.5691 -0.0109 7.7501>, rot=<0.0572 -0.0012 0.9982 0.0155>}
 calf_r {parent=thigh_r, loc=<39.0487 0 0>, rot=<0.0006 -0.0161 -0.0391 -0.9991>}
 foot_r {parent=calf_r, loc=<42.5611 -0.1077 -0.4292>, rot=<0.0418 -0.0455 0.0343 -0.9975>}
 ball_r {parent=foot_r, loc=<6.1598 -13.8264 -0.6806>, rot=<0.0001 -0.0001 -0.7063 -0.7079>}
 thigh_twist_01_r {parent=thigh_r, loc=<21.6303 0 0>, rot=<0.0013 0 0 -1>}
 wolf_thigh_r {parent=pelvis, loc=<-2.5694 -0.0109 7.7501>, rot=<0.0572 -0.0012 0.9982 0.0155>}
 wolf_thigh_twist_01_r {parent=wolf_thigh_r, loc=<21.6303 0 0>, rot=<0.0013 0 0 -1>}
 wolf_calf_r {parent=wolf_thigh_r, loc=<36.4938 -7.6945 -0.7295>, rot=<0.0006 -0.0161 -0.0391 -0.9991>}
 wolf_foot_r {parent=wolf_calf_r, loc=<33.1086 12.6369 0.6355>, rot=<0.0418 -0.0455 0.0343 -0.9975>}
 wolf_ball_r {parent=wolf_foot_r, loc=<16.8627 -7.1236 1.9174>, rot=<0.0001 -0.0001 -0.7063 -0.7079>}
 wolf_toe_r {parent=wolf_ball_r, loc=<-11.4155 -1.9615 0.1853>, rot=<0 0 0 -1>}
 wolf_thigh_l {parent=pelvis, loc=<-2.5693 -0.0109 -7.7501>, rot=<0.0012 0.0572 0.0155 -0.9982>}
 wolf_calf_l {parent=wolf_thigh_l, loc=<-36.4939 7.6945 0.7294>, rot=<0.0006 -0.0161 -0.0391 -0.9991>}
 wolf_foot_l {parent=wolf_calf_l, loc=<-33.1086 -12.6369 -0.6355>, rot=<0.0418 -0.0455 0.0343 -0.9975>}
 wolf_ball_l {parent=wolf_foot_l, loc=<-16.8627 7.1236 -1.9174>, rot=<0 -0.0001 0 -1>}
 wolf_toe_l {parent=wolf_ball_l, loc=<-1.9348 11.4201 -0.1876>, rot=<0.0001 0 -0.7063 -0.7079>}
 wolf_thigh_twist_01_l {parent=wolf_thigh_l, loc=<-21.6303 0 0>, rot=<0.0013 0 0 -1>}
 dyn_tail_socket {parent=pelvis, loc=<-1.2097 -12.2627 -0.0012>, rot=<-0.491 0.5089 0.5089 -0.491>}
]
